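# Post-process AnimBP for the Sydney 3L male medium head, with neck and head pose adjustments for the DimeAviator custom part.
# Customization load detects the part by name wildcard and caches the Face mesh.
# When equipped, updates compute negatively scaled neck and head rotation deltas from ref pose, though the AnimGraph shows no use of them.
# AnimGraph applies fixed bone-space yaw to those bones, behind a blend-by-bool against the input pose with no visible driver.
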AnimBlueprint /Game/Characters/Player/Male/Medium/Heads/M_MED_Sydney3L_Head/Meshes/M_MED_Sydney3L_Head_POST_AnimBP
target_skeleton=/Game/Characters/Player/Common/Fortnite_Base_Head/Fortnite_Base_3L_Head_Skeleton
graphs [
 AnimGraph {nodes [
   Output Pose(Result=Use cached pose 'DimeAviatorMods_Cache'())
   Save cached pose 'CR_Input'(Pose=Input Pose()) {cacheName=CR_Input}
   n1 = "Transform (Modify) Bone - Bone: neck_01"!(ComponentPose=Local To Component(LocalPose=Use cached pose 'CR_Input'()), Alpha=1, Rotation="0,-10,0.000") -> (Pose) {
    rotationMode=BMM_Replace
    rotationSpace=BCS_BoneSpace
    alphaBoolBlend {blendOption=Linear}
    alphaScaleBiasClamp {interpSpeedDecreasing=5, interpSpeedIncreasing=20}
    boneToModify {boneName=neck_01}
    rotation {yaw=-10}
   }
   n2 = "Transform (Modify) Bone - Bone: neck_02"!(ComponentPose=@n1, Alpha=1, Rotation="0,-10,0.000") -> (Pose) {
    rotationMode=BMM_Replace
    rotationSpace=BCS_BoneSpace
    alphaBoolBlend {blendOption=Linear}
    alphaScaleBiasClamp {interpSpeedDecreasing=5, interpSpeedIncreasing=20}
    boneToModify {boneName=neck_02}
    rotation {yaw=-10}
   }
   n3 = "Transform (Modify) Bone - Bone: head"!(ComponentPose=@n2, Alpha=1, Rotation="0,-10,0.000") -> (Pose) {
    rotationMode=BMM_Replace
    rotationSpace=BCS_BoneSpace
    alphaBoolBlend {blendOption=Linear}
    alphaScaleBiasClamp {interpSpeedDecreasing=5, interpSpeedIncreasing=20}
    boneToModify {boneName=head}
    rotation {yaw=-10}
   }
   Save cached pose 'DimeAviatorMods_Cache'(Pose=Blend Poses by bool(BlendPose_0=Component To Local(ComponentPose=@n3), BlendPose_1=Use cached pose 'CR_Input'()) {blendType=Linear}) {cacheName=DimeAviatorMods_Cache}
  ]}
 EventGraph {nodes [
   Event BlueprintUpdateAnimation()
   > n1 = Sequence()
   n2 = Sequence(execute=[@n4.then, @n1.then_0])
   n3 = Get EFort Custom Part Type as Index!(PartType:EFortCustomPartType Enum) -> (ReturnValue:Byte) {is=title K2Node_MacroInstance}
   n4 = Event OnCharacterCustomizationLoaded()
   n5 = Branch(
    execute=@n2.then_0
    Condition=Is Valid Index!(
     TargetArray:Array of Custom Character Part Object References=@n4.CustomParts
     IndexToTest:int="To Integer (Byte)"!(InByte:Byte=@n3) -> (ReturnValue:int=0)
    ) -> (ReturnValue:bool)
   )
   n6 = Get Array Item with EFort Custom Part Type(Array:Array of Custom Character Part Object References=@n4.CustomParts) -> (Output:Custom Character Part Object Reference)
   > Set IsDimeAviatorEquipped(IsDimeAviatorEquipped:bool=MatchesWildcard(SourceString=GetDisplayName(Object=@n6), Wildcard=*DimeAviator*)) -> (Output_Get:bool)
   Set IsDimeAviatorEquipped(execute=@n5.else, IsDimeAviatorEquipped=false)
   n7 = Get EFort Custom Part Type as Index(PartType=Face)
   n8 = Branch(
    execute=@n2.then_1
    Condition=Is Valid Index(
     TargetArray:Array of Skeletal Mesh Component Budgeted Object References=@n4.CustomMeshParts
     IndexToTest="To Integer (Byte)"(InByte=@n7) -> (ReturnValue=0)
    )
   )
   > Set BaseMesh(BaseMesh:Skeletal Mesh Component Object Reference=Get Array Item with EFort Custom Part Type(Array:Array of Skeletal Mesh Component Object References=@n4.CustomMeshParts, PartType=Face) -> (Output:Skeletal Mesh Component Object Reference)) -> (Output_Get:Skeletal Mesh Component Object Reference)
   Set BaseMesh(execute=@n8.else)
   Branch(execute=@n1.then_1, Condition=Get IsDimeAviatorEquipped() -> (IsDimeAviatorEquipped:bool))
   > Set Twist Reduction Value(Twist Reduction Value:double=-0.25) -> (Output_Get:double)
   > Set Y Reduction Value(Y Reduction Value:double=-0.5) -> (Output_Get:double)
   > Set Z Reduction Value(Z Reduction Value:double=-0.5) -> (Output_Get:double)
   n9 = BreakRotator(InRot=GetDeltaTransformFromRefPose!(self=GetOwnerSkeletalMeshComponent(), BoneName=neck_01, BaseName=spine_05) -> (ReturnValue_Location:Vector, ReturnValue_Rotation:Rotator, ReturnValue_Scale:Vector).ReturnValue_Rotation)
   n10 = Get Twist Reduction Value() -> (Twist Reduction Value:double)
   n11 = Get Y Reduction Value() -> (Y Reduction Value:double)
   n12 = Get Z Reduction Value() -> (Z Reduction Value:double)
   > Set neck01 rot(neck01 rot:Rotator=MakeRotator(
     Roll=float * float(A=@n9, B=@n10) {is=title K2Node_PromotableOperator}
     Pitch=float * float(A=@n9.Pitch, B=@n11)
     Yaw=float * float(A=@n9.Yaw, B=@n12)
    )) -> (Output_Get:Rotator)
   n13 = BreakRotator(InRot=GetDeltaTransformFromRefPose(self=GetOwnerSkeletalMeshComponent(), BoneName=neck_02, BaseName=neck_01).ReturnValue_Rotation)
   n14 = Get Twist Reduction Value()
   n15 = Get Y Reduction Value()
   n16 = Get Z Reduction Value()
   > Set neck02 rot(neck02 rot:Rotator=MakeRotator(Roll=float * float(A=@n13, B=@n14), Pitch=float * float(A=@n13.Pitch, B=@n15), Yaw=float * float(A=@n13.Yaw, B=@n16))) -> (Output_Get:Rotator)
   n17 = BreakRotator(InRot=GetDeltaTransformFromRefPose(self=GetOwnerSkeletalMeshComponent(), BoneName=head, BaseName=neck_02).ReturnValue_Rotation)
   n18 = Get Z Reduction Value()
   n19 = Get Twist Reduction Value()
   n20 = Get Y Reduction Value()
   > Set head rot(head rot:Rotator=MakeRotator(Roll=float * float(A=@n17, B=@n19), Pitch=float * float(A=@n17.Pitch, B=@n20), Yaw=float * float(A=@n17.Yaw, B=@n18))) -> (Output_Get:Rotator)
   TryGetPawnOwner()
  ]}
]
variables [
 BaseMesh {type=SkeletalMeshComponent}
 head rot {type=Rotator}
 IsDimeAviatorEquipped {type=boolean}
 neck01 rot {type=Rotator}
 neck02 rot {type=Rotator}
 Twist Reduction Value {type=number}
 Y Reduction Value {type=number}
 Z Reduction Value {type=number}
]
